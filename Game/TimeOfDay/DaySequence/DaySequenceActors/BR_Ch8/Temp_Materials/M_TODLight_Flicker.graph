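# M_TODLight_Flicker is a light function material that makes a light flicker according to time of day.
# A Scalar_Time_of_Day function yields zero for morning, day and evening and one for night, so the light is active only at night.
# That value is multiplied by the sine intensity output of MF_Flicker_IntensityWave to add the flicker.
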Material /Game/TimeOfDay/DaySequence/DaySequenceActors/BR_Ch8/Temp_Materials/M_TODLight_Flicker
graph {
 properties {bAutomaticallySetUsageInEditor=false, materialDomain=MD_LightFunction, usedShadingModels=Default Lit}
 nodes [
  n1 = Constant()
  Multiply(
   A=MaterialFunctionCall(Value_Morning=@n1, Value_Day=@n1, Value_Evening=@n1, Value_Night=Constant() {r=1}) {materialFunction=&Scalar_Time_of_Day}.Result
   B=MaterialFunctionCall() {materialFunction=&/MF_Flicker_IntensityWave}.Intensity Sine
  )
 ]
}
